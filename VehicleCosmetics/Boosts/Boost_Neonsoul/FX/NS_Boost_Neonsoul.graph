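# Niagara effect for the Neonsoul vehicle boost cosmetic, with paint-color variants and placement from car offset and boost position.
# Boost_Main spawns quality-scaled ribbons per port that stream backward, with per-paint color-over-life curves, black secondary color and a second width selected by ColorVariantID.
# Cone_Complete001 bursts persistent sprite and mesh particles for the nozzle cone, using three paint color layers.
# System loop age drives its sprite scale and ring alpha pulse.
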NiagaraSystem /VehicleCosmetics/Boosts/Boost_Neonsoul/FX/NS_Boost_Neonsoul
dependencies {
 usedDataInterfaces [
  &NiagaraDataInterfaceColorCurve
  &NiagaraDataInterfaceCurve
  &NiagaraDataInterfaceArrayFloat
  &NiagaraDataInterfaceSkeletalMesh
 ]
 usedDynamicInputs []
 usedModules [
  &self:SystemSpawnScript.SystemScriptSource.SystemScriptGraph.NiagaraNodeAssignment_0.SetVariables_#1
  &SystemState
 ]
}
emitters {
 Boost_Main {
  emitterUpdateScript [
   &EmitterState(Inactive Response=::NewEnumerator1, Life Cycle Mode=::NewEnumerator1, Loop Duration=0.1, UseLoopDelay=false)
   &VC_FilteredSocketCount()
   &VC_RateQualityHandler(Epic Spawn Rate=50, High Spawn Rate=40, Low Spawn Rate=20, Use Custom Rate?=true)
   &VC_SpawnRibbon_v1(Number of Ribbons=$Emitter.PortCount, SpawnRate=$Emitter.Spawn Rate)
  ]
  particleSpawnScript [
   &VC_PaintedVariantsOverride() {enabled=false}
   &VC_InitializeRibbon_v1(
    Custom Offset=true
    Lifetime=1.2
    Local Vector Offset=&AddVectorToPosition(
     Position=&MakeCustomVectorFromBool(Bool=$User.bIsFrontendPreview, True Vector=<10 0 0>, False Vector=<-5 0 0>)
     Vector=&AddVectorToPosition(Position=$User.DTCarOffset, Vector=Linked() {linkedVariable {name=User.SingleBoostPos, type=NiagaraPosition}})
    )
    Offset Amount=1
    Ribbon Width=120
   )
   &AddVelocity(Velocity=<-1300 0 0>) {module=AddVelocity001}
  ]
  particleUpdateScript [
   &VC_PaintedVariantsOverride(
    Black=&ColorFromCurve(ColorCurve=DataInterface() {propertyValues {
       AlphaCurve {keys [
         {interpMode=RCIM_Cubic, leaveTangentWeight=0.3333, tangentMode=RCTM_User, time=0.02, value=0}
         <at 0.3 0.5>
         {arriveTangentWeight=0.3333, interpMode=RCIM_Cubic, tangentMode=RCTM_User, time=0.9, value=0}
        ]}
       BlueCurve {keys [{interpMode=RCIM_Cubic, leaveTangentWeight=0.3333, tangentMode=RCTM_User, time=0, value=0.2}, <at 0.8248 0.02>]}
       GreenCurve {keys [{interpMode=RCIM_Cubic, leaveTangentWeight=0.3333, tangentMode=RCTM_User, time=0, value=0.2}, <at 0.8248 0.02>]}
       RedCurve {keys [{interpMode=RCIM_Cubic, leaveTangentWeight=0.3333, tangentMode=RCTM_User, time=0, value=0.2}, <at 0.8248 0.02>]}
      }})
    BurntSienna=&ColorFromCurve(ColorCurve=DataInterface() {propertyValues {
       AlphaCurve {keys [
         {interpMode=RCIM_Cubic, leaveTangentWeight=0.3333, tangentMode=RCTM_User, time=0.02, value=0}
         <at 0.3 0.8>
         {arriveTangentWeight=0.3333, interpMode=RCIM_Cubic, tangentMode=RCTM_User, time=0.9, value=0}
        ]}
       BlueCurve {keys [{interpMode=RCIM_Cubic, leaveTangentWeight=0.3333, tangentMode=RCTM_User, time=0, value=0.055}, <at 0.5 0.03>]}
       GreenCurve {keys [{interpMode=RCIM_Cubic, leaveTangentWeight=0.3333, tangentMode=RCTM_User, time=0, value=0.2292}, <at 0.5 0.05>]}
       RedCurve {keys [{interpMode=RCIM_Cubic, leaveTangentWeight=0.3333, tangentMode=RCTM_User, time=0, value=1.1}, <at 0.5 0.15>]}
      }})
    Cobalt=&ColorFromCurve(ColorCurve=DataInterface() {propertyValues {
       AlphaCurve {keys [
         {interpMode=RCIM_Cubic, leaveTangentWeight=0.3333, tangentMode=RCTM_User, time=0.02, value=0}
         <at 0.3 0.9>
         {arriveTangentWeight=0.3333, interpMode=RCIM_Cubic, tangentMode=RCTM_User, time=0.9, value=0}
        ]}
       BlueCurve {keys [{interpMode=RCIM_Cubic, leaveTangentWeight=0.3333, tangentMode=RCTM_User, time=0, value=2}, <at 0.5 0.8>]}
       GreenCurve {keys [{interpMode=RCIM_Cubic, leaveTangentWeight=0.3333, tangentMode=RCTM_User, time=0, value=0.295}, <at 0.5 0.0683>]}
       RedCurve {keys [{interpMode=RCIM_Cubic, leaveTangentWeight=0.3333, tangentMode=RCTM_User, time=0, value=0.14}, <at 0.5 0.016>]}
      }})
    Crimson=&ColorFromCurve(ColorCurve=DataInterface() {propertyValues {
       AlphaCurve {keys [
         {interpMode=RCIM_Cubic, leaveTangentWeight=0.3333, tangentMode=RCTM_User, time=0.02, value=0}
         <at 0.3 0.8>
         {arriveTangentWeight=0.3333, interpMode=RCIM_Cubic, tangentMode=RCTM_User, time=0.9, value=0}
        ]}
       BlueCurve {keys [{interpMode=RCIM_Cubic, leaveTangentWeight=0.3333, tangentMode=RCTM_User, time=0, value=0.104}, <at 0.5 0.04>]}
       GreenCurve {keys [{interpMode=RCIM_Cubic, leaveTangentWeight=0.3333, tangentMode=RCTM_User, time=0, value=0.104}, <at 0.5 0.04>]}
       RedCurve {keys [{interpMode=RCIM_Cubic, leaveTangentWeight=0.3333, tangentMode=RCTM_User, time=0, value=1.3}, <at 0.5 0.5>]}
      }})
    DefaultColor=&ColorFromCurve(ColorCurve=DataInterface() {propertyValues {
       AlphaCurve {keys [
         {interpMode=RCIM_Cubic, leaveTangentWeight=0.3333, tangentMode=RCTM_User, time=0.02, value=0}
         <at 0.3 0.9>
         {arriveTangentWeight=0.3333, interpMode=RCIM_Cubic, tangentMode=RCTM_User, time=0.9, value=0}
        ]}
       BlueCurve {keys [{interpMode=RCIM_Cubic, leaveTangentWeight=0.3333, tangentMode=RCTM_User, time=0, value=0.026}, <at 0.5 0.3893>]}
       GreenCurve {keys [{interpMode=RCIM_Cubic, leaveTangentWeight=0.3333, tangentMode=RCTM_User, time=0, value=0.0472}, <at 0.5 0.0514>]}
       RedCurve {keys [{interpMode=RCIM_Cubic, leaveTangentWeight=0.3333, tangentMode=RCTM_User, time=0, value=1.3}, <at 0.5 0.4115>]}
      }})
    ForestGreen=&ColorFromCurve(ColorCurve=DataInterface() {propertyValues {
       AlphaCurve {keys [
         {interpMode=RCIM_Cubic, leaveTangentWeight=0.3333, tangentMode=RCTM_User, time=0.02, value=0}
         <at 0.3 0.9>
         {arriveTangentWeight=0.3333, interpMode=RCIM_Cubic, tangentMode=RCTM_User, time=0.9, value=0}
        ]}
       BlueCurve {keys [{interpMode=RCIM_Cubic, leaveTangentWeight=0.3333, tangentMode=RCTM_User, time=0, value=0.024}, <at 0.5 0.0032>]}
       GreenCurve {keys [{interpMode=RCIM_Cubic, leaveTangentWeight=0.3333, tangentMode=RCTM_User, time=0, value=0.6}, <at 0.5 0.16>]}
       RedCurve {keys [{interpMode=RCIM_Cubic, leaveTangentWeight=0.3333, tangentMode=RCTM_User, time=0, value=0.024}, <at 0.5 0.0032>]}
      }})
    Grey=&ColorFromCurve(ColorCurve=DataInterface() {propertyValues {
       AlphaCurve {keys [
         {interpMode=RCIM_Cubic, leaveTangentWeight=0.3333, tangentMode=RCTM_User, time=0.02, value=0}
         <at 0.3 0.8>
         {arriveTangentWeight=0.3333, interpMode=RCIM_Cubic, tangentMode=RCTM_User, time=0.9, value=0}
        ]}
       BlueCurve {keys [{interpMode=RCIM_Cubic, leaveTangentWeight=0.3333, tangentMode=RCTM_User, time=0, value=0.3}, <at 0.8223 0.04>]}
       GreenCurve {keys [{interpMode=RCIM_Cubic, leaveTangentWeight=0.3333, tangentMode=RCTM_User, time=0, value=0.3}, <at 0.8223 0.04>]}
       RedCurve {keys [{interpMode=RCIM_Cubic, leaveTangentWeight=0.3333, tangentMode=RCTM_User, time=0, value=0.3}, <at 0.8223 0.04>]}
      }})
    Lime=&ColorFromCurve(ColorCurve=DataInterface() {propertyValues {
       AlphaCurve {keys [
         {interpMode=RCIM_Cubic, leaveTangentWeight=0.3333, tangentMode=RCTM_User, time=0.02, value=0}
         <at 0.3 0.9>
         {arriveTangentWeight=0.3333, interpMode=RCIM_Cubic, tangentMode=RCTM_User, time=0.9, value=0}
        ]}
       BlueCurve {keys [{interpMode=RCIM_Cubic, leaveTangentWeight=0.3333, tangentMode=RCTM_User, time=0, value=0.016}, <at 0.5 0.004>]}
       GreenCurve {keys [{interpMode=RCIM_Cubic, leaveTangentWeight=0.3333, tangentMode=RCTM_User, time=0, value=0.8}, <at 0.5 0.2>]}
       RedCurve {keys [{interpMode=RCIM_Cubic, leaveTangentWeight=0.3333, tangentMode=RCTM_User, time=0, value=0.4733}, <at 0.5 0.1183>]}
      }})
    Orange=&ColorFromCurve(ColorCurve=DataInterface() {propertyValues {
       AlphaCurve {keys [
         {interpMode=RCIM_Cubic, leaveTangentWeight=0.3333, tangentMode=RCTM_User, time=0.02, value=0}
         <at 0.3 0.8>
         {arriveTangentWeight=0.3333, interpMode=RCIM_Cubic, tangentMode=RCTM_User, time=0.9, value=0}
        ]}
       BlueCurve {keys [{interpMode=RCIM_Cubic, leaveTangentWeight=0.3333, tangentMode=RCTM_User, time=0, value=0.015}, <at 0.5 0.024>]}
       GreenCurve {keys [{interpMode=RCIM_Cubic, leaveTangentWeight=0.3333, tangentMode=RCTM_User, time=0, value=0.2625}, <at 0.5 0.1008>]}
       RedCurve {keys [{interpMode=RCIM_Cubic, leaveTangentWeight=0.3333, tangentMode=RCTM_User, time=0, value=1.5}, <at 0.5 0.6>]}
      }})
    Pink=&ColorFromCurve(ColorCurve=DataInterface() {propertyValues {
       AlphaCurve {keys [
         {interpMode=RCIM_Cubic, leaveTangentWeight=0.3333, tangentMode=RCTM_User, time=0.02, value=0}
         <at 0.3 0.8>
         {arriveTangentWeight=0.3333, interpMode=RCIM_Cubic, tangentMode=RCTM_User, time=0.9, value=0}
        ]}
       BlueCurve {keys [{interpMode=RCIM_Cubic, leaveTangentWeight=0.3333, tangentMode=RCTM_User, time=0, value=1.2133}, <at 0.5 0.6933>]}
       GreenCurve {keys [{interpMode=RCIM_Cubic, leaveTangentWeight=0.3333, tangentMode=RCTM_User, time=0, value=0.28}, <at 0.5 0.16>]}
       RedCurve {keys [{interpMode=RCIM_Cubic, leaveTangentWeight=0.3333, tangentMode=RCTM_User, time=0, value=1.4}, <at 0.5 0.8>]}
      }})
    Purple=&ColorFromCurve(ColorCurve=DataInterface() {propertyValues {
       AlphaCurve {keys [
         {interpMode=RCIM_Cubic, leaveTangentWeight=0.3333, tangentMode=RCTM_User, time=0.02, value=0}
         <at 0.3 0.8>
         {arriveTangentWeight=0.3333, interpMode=RCIM_Cubic, tangentMode=RCTM_User, time=0.9, value=0}
        ]}
       BlueCurve {keys [{interpMode=RCIM_Cubic, leaveTangentWeight=0.3333, tangentMode=RCTM_User, time=0, value=2}, <at 0.5 0.6>]}
       GreenCurve {keys [{interpMode=RCIM_Cubic, leaveTangentWeight=0.3333, tangentMode=RCTM_User, time=0, value=0.3}, <at 0.5 0.03>]}
       RedCurve {keys [{interpMode=RCIM_Cubic, leaveTangentWeight=0.3333, tangentMode=RCTM_User, time=0, value=0.6967}, <at 0.5 0.087>]}
      }})
    R1 Black:LinearColor=&ColorFromCurve(ColorCurve=DataInterface() {propertyValues {
       AlphaCurve {keys [<at 0.02 0>, <at 0.3 2>, {arriveTangentWeight=0.3333, interpMode=RCIM_Cubic, tangentMode=RCTM_User, time=0.8, value=0}]}
       BlueCurve {keys [<at 0 0>]}
       GreenCurve {keys [<at 0 0>]}
       RedCurve {keys [<at 0 0>]}
      }})
    R1 BurntSienna:LinearColor=&ColorFromCurve(ColorCurve=DataInterface() {propertyValues {
       AlphaCurve {keys [
         {interpMode=RCIM_Cubic, leaveTangentWeight=0.3333, tangentMode=RCTM_User, time=0.02, value=0}
         <at 0.3 2>
         {arriveTangentWeight=0.3333, interpMode=RCIM_Cubic, tangentMode=RCTM_User, time=0.8, value=0}
        ]}
       BlueCurve {keys [<at 0 0>]}
       GreenCurve {keys [<at 0 0>]}
       RedCurve {keys [<at 0 0>]}
      }})
    R1 Cobalt:LinearColor=&ColorFromCurve(ColorCurve=DataInterface() {propertyValues {
       AlphaCurve {keys [<at 0.02 0>, <at 0.3 2>, {arriveTangentWeight=0.3333, interpMode=RCIM_Cubic, tangentMode=RCTM_User, time=0.8, value=0}]}
       BlueCurve {keys [<at 0 0>]}
       GreenCurve {keys [<at 0 0>]}
       RedCurve {keys [<at 0 0>]}
      }})
    R1 Crimson:LinearColor=&ColorFromCurve(ColorCurve=DataInterface() {propertyValues {
       AlphaCurve {keys [<at 0.02 0>, <at 0.3 2>, {arriveTangentWeight=0.3333, interpMode=RCIM_Cubic, tangentMode=RCTM_User, time=0.8, value=0}]}
       BlueCurve {keys [<at 0 0>]}
       GreenCurve {keys [<at 0 0>]}
       RedCurve {keys [<at 0 0>]}
      }})
    R1 DefaultColor=&ColorFromCurve(ColorCurve=DataInterface() {propertyValues {
       AlphaCurve {keys [<at 0.02 0>, <at 0.3 2>, {arriveTangentWeight=0.3333, interpMode=RCIM_Cubic, tangentMode=RCTM_User, time=0.8, value=0}]}
       BlueCurve {keys [<at 0 0>]}
       GreenCurve {keys [<at 0 0>]}
       RedCurve {keys [<at 0 0>]}
      }})
    R1 ForestGreen:LinearColor=&ColorFromCurve(ColorCurve=DataInterface() {propertyValues {
       AlphaCurve {keys [<at 0.02 0>, <at 0.3 2>, {arriveTangentWeight=0.3333, interpMode=RCIM_Cubic, tangentMode=RCTM_User, time=0.8, value=0}]}
       BlueCurve {keys [<at 0 0>]}
       GreenCurve {keys [<at 0 0>]}
       RedCurve {keys [<at 0 0>]}
      }})
    R1 Gold:LinearColor=&ColorFromCurve(ColorCurve=DataInterface() {propertyValues {
       AlphaCurve {keys [<at 0.02 0>, <at 0.3 2>, {arriveTangentWeight=0.3333, interpMode=RCIM_Cubic, tangentMode=RCTM_User, time=0.8, value=0}]}
       BlueCurve {keys [<at 0 0>]}
       GreenCurve {keys [<at 0 0>]}
       RedCurve {keys [<at 0 0>]}
      }})
    R1 Grey:LinearColor=&ColorFromCurve(ColorCurve=DataInterface() {propertyValues {
       AlphaCurve {keys [<at 0.02 0>, <at 0.3 2>, {arriveTangentWeight=0.3333, interpMode=RCIM_Cubic, tangentMode=RCTM_User, time=0.8, value=0}]}
       BlueCurve {keys [<at 0 0>]}
       GreenCurve {keys [<at 0 0>]}
       RedCurve {keys [<at 0 0>]}
      }})
    R1 Lime:LinearColor=&ColorFromCurve(ColorCurve=DataInterface() {propertyValues {
       AlphaCurve {keys [<at 0.02 0>, <at 0.3 2>, {arriveTangentWeight=0.3333, interpMode=RCIM_Cubic, tangentMode=RCTM_User, time=0.8, value=0}]}
       BlueCurve {keys [<at 0 0>]}
       GreenCurve {keys [<at 0 0>]}
       RedCurve {keys [<at 0 0>]}
      }})
    R1 Orange:LinearColor=&ColorFromCurve(ColorCurve=DataInterface() {propertyValues {
       AlphaCurve {keys [<at 0.02 0>, <at 0.3 2>, {arriveTangentWeight=0.3333, interpMode=RCIM_Cubic, tangentMode=RCTM_User, time=0.8, value=0}]}
       BlueCurve {keys [<at 0 0>]}
       GreenCurve {keys [<at 0 0>]}
       RedCurve {keys [<at 0 0>]}
      }})
    R1 Pink:LinearColor=&ColorFromCurve(ColorCurve=DataInterface() {propertyValues {
       AlphaCurve {keys [<at 0.02 0>, <at 0.3 2>, {arriveTangentWeight=0.3333, interpMode=RCIM_Cubic, tangentMode=RCTM_User, time=0.8, value=0}]}
       BlueCurve {keys [<at 0 0>]}
       GreenCurve {keys [<at 0 0>]}
       RedCurve {keys [<at 0 0>]}
      }})
    R1 Purple:LinearColor=&ColorFromCurve(ColorCurve=DataInterface() {propertyValues {
       AlphaCurve {keys [<at 0.02 0>, <at 0.3 2>, {arriveTangentWeight=0.3333, interpMode=RCIM_Cubic, tangentMode=RCTM_User, time=0.8, value=0}]}
       BlueCurve {keys [<at 0 0>]}
       GreenCurve {keys [<at 0 0>]}
       RedCurve {keys [<at 0 0>]}
      }})
    R1 Saffron:LinearColor=&ColorFromCurve(ColorCurve=DataInterface() {propertyValues {
       AlphaCurve {keys [<at 0.02 0>, <at 0.3 2>, {arriveTangentWeight=0.3333, interpMode=RCIM_Cubic, tangentMode=RCTM_User, time=0.8, value=0}]}
       BlueCurve {keys [<at 0 0>]}
       GreenCurve {keys [<at 0 0>]}
       RedCurve {keys [<at 0 0>]}
      }})
    R1 SkyBlue:LinearColor=&ColorFromCurve(ColorCurve=DataInterface() {propertyValues {
       AlphaCurve {keys [<at 0.02 0>, <at 0.3 2>, {arriveTangentWeight=0.3333, interpMode=RCIM_Cubic, tangentMode=RCTM_User, time=0.8, value=0}]}
       BlueCurve {keys [<at 0 0>]}
       GreenCurve {keys [<at 0 0>]}
       RedCurve {keys [<at 0 0>]}
      }})
    R1 TitaniumWhite:LinearColor=&ColorFromCurve(ColorCurve=DataInterface() {propertyValues {
       AlphaCurve {keys [<at 0.02 0>, <at 0.3 2>, {arriveTangentWeight=0.3333, interpMode=RCIM_Cubic, tangentMode=RCTM_User, time=0.8, value=0}]}
       BlueCurve {keys [<at 0 0>]}
       GreenCurve {keys [<at 0 0>]}
       RedCurve {keys [<at 0 0>]}
      }})
    Saffron=&ColorFromCurve(ColorCurve=DataInterface() {propertyValues {
       AlphaCurve {keys [
         {interpMode=RCIM_Cubic, leaveTangentWeight=0.3333, tangentMode=RCTM_User, time=0.02, value=0}
         <at 0.3 0.8>
         {arriveTangentWeight=0.3333, interpMode=RCIM_Cubic, tangentMode=RCTM_User, time=0.9, value=0}
        ]}
       BlueCurve {keys [{interpMode=RCIM_Cubic, leaveTangentWeight=0.3333, tangentMode=RCTM_User, time=0, value=0.072}, <at 0.5 0.03>]}
       GreenCurve {keys [{interpMode=RCIM_Cubic, leaveTangentWeight=0.3333, tangentMode=RCTM_User, time=0, value=0.693}, <at 0.5 0.41>]}
       RedCurve {keys [{interpMode=RCIM_Cubic, leaveTangentWeight=0.3333, tangentMode=RCTM_User, time=0, value=0.9}, <at 0.5 0.6>]}
      }})
    SkyBlue=&ColorFromCurve(ColorCurve=DataInterface() {propertyValues {
       AlphaCurve {keys [
         {interpMode=RCIM_Cubic, leaveTangentWeight=0.3333, tangentMode=RCTM_User, time=0.02, value=0}
         <at 0.3 0.9>
         {arriveTangentWeight=0.3333, interpMode=RCIM_Cubic, tangentMode=RCTM_User, time=0.9, value=0}
        ]}
       BlueCurve {keys [{interpMode=RCIM_Cubic, leaveTangentWeight=0.3333, tangentMode=RCTM_User, time=0, value=1.3}, <at 0.5 0.8>]}
       GreenCurve {keys [{interpMode=RCIM_Cubic, leaveTangentWeight=0.3333, tangentMode=RCTM_User, time=0, value=0.4972}, <at 0.5 0.2773>]}
       RedCurve {keys [{interpMode=RCIM_Cubic, leaveTangentWeight=0.3333, tangentMode=RCTM_User, time=0, value=0.065}, <at 0.5 0.016>]}
      }})
    TitaniumWhite=&ColorFromCurve(ColorCurve=DataInterface() {propertyValues {
       AlphaCurve {keys [
         {interpMode=RCIM_Cubic, leaveTangentWeight=0.3333, tangentMode=RCTM_User, time=0.02, value=0}
         <at 0.3 0.8>
         {arriveTangentWeight=0.3333, interpMode=RCIM_Cubic, tangentMode=RCTM_User, time=0.9, value=0}
        ]}
       BlueCurve {keys [{interpMode=RCIM_Cubic, leaveTangentWeight=0.3333, tangentMode=RCTM_User, time=0, value=0.85}, <at 0.5512 0.3>]}
       GreenCurve {keys [{interpMode=RCIM_Cubic, leaveTangentWeight=0.3333, tangentMode=RCTM_User, time=0, value=0.85}, <at 0.5512 0.3>]}
       RedCurve {keys [{interpMode=RCIM_Cubic, leaveTangentWeight=0.3333, tangentMode=RCTM_User, time=0, value=0.85}, <at 0.5512 0.3>]}
      }})
    Use R1 Painted Overrides=true
   ) {module=VC_PaintedVariantsOverride001}
   &ParticleState()
   &InheritSourceMovement(ApplyForce=false, ApplyPosition=true, ApplyPositionFalloff=false)
   &SolveForcesAndVelocity(Clamp Velocity=false, Limit Acceleration=false)
   &VC_UpdateRibbon_v1(Enable Skeletal Position=true)
   &ScaleRibbonWidth(Ribbon Width Scale=&FloatFromCurve(FloatCurve=DataInterface() {propertyValues {Curve {keys [
         {arriveTangent=0.3849, interpMode=RCIM_Cubic, leaveTangent=0.3849, leaveTangentWeight=0.3333, tangentMode=RCTM_User, time=0, value=1}
         {arriveTangent=-0.0697, arriveTangentWeight=0.3333, interpMode=RCIM_Cubic, leaveTangent=-0.0697, tangentMode=RCTM_User, time=1, value=0.6}
        ]}}}))
   /SetVariables(Particles.width2:NiagaraFloat=&Multiply_Float(
     A=&FloatFromCurve(FloatCurve=DataInterface() {propertyValues {Curve {keys [
          {arriveTangent=0.3849, interpMode=RCIM_Cubic, leaveTangent=0.3849, leaveTangentWeight=0.3333, tangentMode=RCTM_User, time=0, value=1}
          {arriveTangent=-0.0697, arriveTangentWeight=0.3333, interpMode=RCIM_Cubic, leaveTangent=-0.0697, tangentMode=RCTM_User, time=1, value=0.6}
         ]}}})
     B=&SelectFloatFromArray(
      Direct Array Index=$User.ColorVariantID
      Float Selection Array=DataInterface() {propertyValues {FloatData [130,130,140,140,130,132,135,140,140,140,135,130,140,135,135]}}
     )
    )) {isSetParameters=true}
  ]
  rendererClasses [&NiagaraRibbonRendererProperties]
  renderers [
   {rendererClass=&NiagaraRibbonRendererProperties}
   {rendererClass=&NiagaraRibbonRendererProperties, rendererIndex=1}
  ]
 }
 Cone_Complete001 {
  emitterUpdateScript [
   &EmitterState(
    Inactive Response=::NewEnumerator1
    Life Cycle Mode=::NewEnumerator1
    Loop Behavior=::NewEnumerator1
    Loop Duration Mode=::NewEnumerator1
    UseLoopDelay=false
   )
   &VC_FilteredSocketCount() {enabled=false}
   &SpawnRate(SpawnGroup=1, SpawnRate=10, Use Spawn Probability=false) {enabled=false}
   &SpawnBurst_Instantaneous(Use Spawn Probability=false, UseLoopCountLimit=false)
  ]
  particleSpawnScript [
   &VC_PaintedVariantsOverride(
    Black=<rgba 0.01 0.01 0.01 0.8>
    BurntSienna=<rgba 0.04 0.0073 0.0008 0.8>
    Cobalt=<rgba 0 0.0083 0.1 0.8>
    Crimson=<rgba 0.4 0 0 0.8>
    DefaultColor=<rgba 0.4 0 0 0.8>
    ForestGreen=<rgba 0 0.1 0 0.8>
    Grey=<rgba 0.05 0.05 0.05 0.8>
    Lime=<rgba 0.0583 0.1 0 0.8>
    Orange=<rgba 0.3 0.025 0 0.8>
    Pink=<rgba 0.2 0 0.1667 0.8>
    Purple=<rgba 0.025 0 0.1 0.8>
    R1 Black:LinearColor=<rgba 0.25 0.25 0.25 0.4>
    R1 BurntSienna:LinearColor=<rgba 2 0.5 0.2 0.4>
    R1 Cobalt:LinearColor=<rgba 0.2 0.5 2 0.8>
    R1 Crimson:LinearColor=<rgba 2 0.385 0.3 0.4>
    R1 DefaultColor=<rgba 2 0.29 0.2 0.4>
    R1 ForestGreen:LinearColor=<rgba 0.4 2 0.4 0.4>
    R1 Grey:LinearColor=<rgba 1 1 1 0.4>
    R1 Lime:LinearColor=<rgba 0.875 1.5 0 0.6>
    R1 Orange:LinearColor=<rgba 2 0.4 0 0.8>
    R1 Pink:LinearColor=<rgba 2 0.4 1.7333 0.8>
    R1 Purple:LinearColor=<rgba 0.9333 0.4 2 0.8>
    R1 Saffron:LinearColor=<rgba 2 1.5 0 0.4>
    R1 SkyBlue:LinearColor=<rgba 0.2 0.8 2 0.8>
    R1 TitaniumWhite:LinearColor=<rgba 2 2 2 0.4>
    R2 Black:LinearColor=<rgba 0 0 0 2>
    R2 BurntSienna:LinearColor=<rgba 0 0 0 2>
    R2 Cobalt:LinearColor=<rgba 0 0 0 2>
    R2 Crimson:LinearColor=<rgba 0 0 0 2>
    R2 DefaultColor:LinearColor=<rgba 0 0 0 2>
    R2 ForestGreen:LinearColor=<rgba 0 0 0 2>
    R2 Gold:LinearColor=<rgba 0 0 0 2>
    R2 Grey:LinearColor=<rgba 0 0 0 2>
    R2 Lime:LinearColor=<rgba 0 0 0 2>
    R2 Orange:LinearColor=<rgba 0 0 0 2>
    R2 Pink:LinearColor=<rgba 0 0 0 2>
    R2 Purple:LinearColor=<rgba 0 0 0 2>
    R2 Saffron:LinearColor=<rgba 0 0 0 2>
    R2 SkyBlue:LinearColor=<rgba 0 0 0 2>
    R2 TitaniumWhite:LinearColor=<rgba 0 0 0 2>
    Saffron=<rgba 0.1 0.075 0 0.8>
    SkyBlue=<rgba 0 0.0333 0.1 0.8>
    TitaniumWhite=<rgba 0.1 0.1 0.1 0.6>
    Use R1 Painted Overrides=true
    Use R2 Painted Overrides=true
   )
   &InitializeParticle(
    Color=$Particles.Color
    Mesh Scale=&Multiply_VectorByFloat(Vector=<0.9 1.4 0.9>, Float=1.05)
    Mesh Scale Mode=::NewEnumerator4
    Position Offset=&AddVectorToPosition(
     Position=<-65 0 0>
     Vector=&AddVectorToPosition(Position=$User.DTCarOffset, Vector=Linked() {linkedVariable {name=User.SingleBoostPos, type=NiagaraPosition}})
    )
    Sprite Rotation Mode=::NewEnumerator2
    Sprite Size Mode=::NewEnumerator3
    Uniform Sprite Size=50
    UsePositionOffset=true
    Write Material Random=true
   )
   &PartitionParticles(Number Of Partitions=$Emitter.PortCount)
   &SkeletalMeshLocation(
    Bone / Socket Index=$Particles.PartitionParticles.Partition
    Bone / Socket Sampling Mode=::NewEnumerator5
    MaskSpawnGroup=false
    Mesh Sampling Type=::NewEnumerator6
    Skeletal Mesh=$System.SkeletalMesh
   ) {enabled=false}
   &AddVelocity(Velocity=<0 0 0>) {enabled=false}
   &InitialMeshOrientation(Mesh Axis To Orient=<0 0 0>, Mesh Orientation Mode=::NewEnumerator2, Orientation Vector=<0 -1 0>, Use Rotation Vector=false)
   /SetVariables(
    Particles.sizeglow:Vector2f=&Vector2DFromFloat(Value=150)
    Particles.spotcolor:LinearColor=&MultiplyLinearColor(
     Linear Color=$Particles.R1 Color
     Scale Factor=&MakeVector4FromVectorAndFloat("Vector (RGB)"=&VectorFromFloat(Value=0.5), "Float (Alpha)"=0.8)
    )
   ) {isSetParameters=true}
  ]
  particleUpdateScript [
   &ParticleState(Kill Particles When Lifetime Has Elapsed=false)
   &ScaleSpriteSize(
    Uniform Curve Index=$System.NormalizedLoopAge
    Uniform Curve Sprite Scale=DataInterface() {propertyValues {Curve {defaultValue=0.8, keys [<at 0 0.2 RCIM_Cubic>,<at 1 1 RCIM_Cubic>]}}}
   )
   &SolveForcesAndVelocity(Clamp Velocity=false, Limit Acceleration=false) {enabled=false}
   /SetVariables(Particles.ringcolor:LinearColor=&MultiplyLinearColor(
     Linear Color=$Particles.R1 Color
     Scale Factor=&MakeVector4FromVectorAndFloat(
      "Vector (RGB)"=&VectorFromFloat(Value=1)
      "Float (Alpha)"=&FloatFromCurve(
       FloatCurve=DataInterface() {propertyValues {Curve {
          defaultValue=1
          keys [
           {interpMode=RCIM_Cubic, tangentMode=RCTM_User, time=0, value=0}
           {interpMode=RCIM_Cubic, tangentMode=RCTM_User, time=0.5, value=1}
           {tangentMode=RCTM_User, time=1, value=0}
          ]
         }}}
       CurveIndex=$System.NormalizedLoopAge
      )
     )
    )) {isSetParameters=true}
  ]
  rendererClasses [&NiagaraSpriteRendererProperties, &NiagaraMeshRendererProperties]
  renderers [
   {rendererClass=&NiagaraSpriteRendererProperties}
   {rendererClass=&NiagaraMeshRendererProperties, rendererIndex=1}
   {rendererClass=&NiagaraMeshRendererProperties, rendererIndex=2}
   {rendererClass=&NiagaraMeshRendererProperties, rendererIndex=3}
   {rendererClass=&NiagaraMeshRendererProperties, rendererIndex=4}
   {rendererClass=&NiagaraSpriteRendererProperties, rendererIndex=5}
  ]
 }
}
summary {
 systemName=NS_Boost_Neonsoul
 emitters [
  {emitterName=Boost_Main, rendererClasses [&NiagaraRibbonRendererProperties]}
  {
   emitterName=Cone_Complete001
   rendererClasses [&NiagaraSpriteRendererProperties, &NiagaraMeshRendererProperties]
  }
 ]
 userVariables [
  User.bIsFrontendPreview {type=NiagaraBool}
  User.bIsPainted {type=NiagaraBool}
  User.ColorVariantID {type=NiagaraInt32}
  User.DTCarOffset {type=NiagaraPosition, description="", defaultValue {}}
  User.PaintedColor {type=LinearColor, description="", defaultValue {struct=LinearColor, value=<rgba 1 1 1 1>}}
  User.PaintedScalar {type=NiagaraFloat, defaultValue=1, description=""}
  User.PortCount {type=NiagaraInt32, defaultValue=1}
  User.SingleBoostPos {type=NiagaraPosition, description="", defaultValue {}}
 ]
}
